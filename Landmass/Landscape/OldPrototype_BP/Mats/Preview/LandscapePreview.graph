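# LandscapePreview is a Default Lit preview material from an old Landmass landscape prototype that shows painted landscape layers as flat colors.
# A chain of lerps starts from a scalar Base value and blends in Farm, Desert, Snow and Mountains color parameters, each weighted by its landscape layer sample, with Mountains applied last.
# It uses no textures, declares scalar parameters for Roughness, Spec and Opacity, and turns tangent-space normals off.
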Material /Landmass/Landscape/OldPrototype_BP/Mats/Preview/LandscapePreview
graph {
 properties {bTangentSpaceNormal=false, pixelDepthOffsetMode=PDOM_Legacy, usedShadingModels=Default Lit}
 nodes [
  n1 = LinearInterpolate(
   A=ScalarParameter() {defaultValue=0.185, parameterName=Base}
   B=VectorParameter() {parameterName=Farm, defaultValue=<rgba 0.07 0.1719 0.0638 1>}.RGB
   Alpha=LandscapeLayerSample() {parameterName=Farm}
  )
  n2 = VectorParameter() {parameterName=DesertColor, defaultValue=<rgba 0.9688 0.4848 0.2069 1>}
  n3 = LandscapeLayerSample() {parameterName=Desert}
  LinearInterpolate(
   A=LinearInterpolate(
    A=LinearInterpolate(A=@n1, B=@n2.RGB, Alpha=@n3)
    B=VectorParameter() {parameterName=Snow, defaultValue=<rgba 0.8 0.9 0.9 1>}.RGB
    Alpha=LandscapeLayerSample() {parameterName=Snow}
   )
   B=VectorParameter() {parameterName=Mountains, defaultValue=<rgba 0.1 0.1 0.1 1>}.RGB
   Alpha=LandscapeLayerSample() {parameterName=Mountains}
  )
  ScalarParameter() {defaultValue=0.75, parameterName=Roughness}
  ScalarParameter() {defaultValue=1, parameterName=Opacity}
  ScalarParameter() {defaultValue=0.25, parameterName=Spec}
 ]
}
