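# Material function that blends Bottom and Top material attribute layers by height, in an old Landmass landscape prototype.
# Each layer's Opacity serves as height; their scaled difference, clamped at zero, feeds HeightLerp with the saturated Alpha and Contrast.
# Top base color is multiplied by an edge tint that fades toward white as Alpha rises.
# Both Blend Vertex Shader branches are identical, keeping Bottom's vertex attributes.
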MaterialFunction /Landmass/Landscape/OldPrototype_BP/Mats/MF/LayerHeightBlend
graph {nodes [
  n1 = GetMaterialAttributes(None=FunctionInput() {inputName=Bottom, inputType=FunctionInput_MaterialAttributes}) {attributeGetTypes [#1]}
  n2 = GetMaterialAttributes(None=FunctionInput() {inputName=Top, inputType=FunctionInput_MaterialAttributes, sortPriority=1}) {attributeGetTypes [#1, #2]}
  n3 = FunctionInput(Preview=VectorParameter() {parameterName=TintColor, defaultValue=<rgba 1 1 1 1>}.RGB) {bUsePreviewValueAsDefault=true, inputName=Edge Tint Color, sortPriority=3}
  n4 = Saturate(None=FunctionInput() {inputName=Alpha, inputType=FunctionInput_Scalar, sortPriority=2})
  n5 = SetMaterialAttributes!(MaterialAttributes=@n2.MaterialAttributes, BaseColor=Multiply(A=@n2.BaseColor, B=LinearInterpolate(A=@n3, Alpha=@n4))) {attributeSetTypes [#2]}
  n6 = FunctionInput() {bUsePreviewValueAsDefault=true, inputName=Height Multiplier, inputType=FunctionInput_Scalar, sortPriority=5, previewValue=<1 0 0 1>}
  n7 = Subtract(B=Multiply(A=ComponentMask(None=@n1.Opacity) {r=true}, B=@n6)) {constA=0.5, constB=0.5}
  n8 = Subtract(A=Multiply(A=ComponentMask(None=@n2.Opacity) {r=true}, B=@n6)) {constB=0.5}
  n9 = MaterialFunctionCall(
   Transition Phase=@n4
   Height Texture=Max(A=Add(A=@n7, B=@n8)) {constB=0}
   Contrast=FunctionInput(Preview=ScalarParameter() {defaultValue=1, parameterName=Preview Height Blend Contrast}) {bUsePreviewValueAsDefault=true, inputName=Contrast, inputType=FunctionInput_Scalar, sortPriority=5}
  ) {materialFunction=&HeightLerp}
  n10 = StaticBool()
  FunctionOutput(None=StaticSwitch(
    True=BlendMaterialAttributes(A=@n1.MaterialAttributes, B=@n5, Alpha=@n9.Alpha) {vertexAttributeBlendType=UseA}
    False=BlendMaterialAttributes(A=@n1.MaterialAttributes, B=@n5, Alpha=@n9.Alpha) {vertexAttributeBlendType=UseA}
    Value=FunctionInput(Preview=@n10) {bUsePreviewValueAsDefault=true, inputName=Blend Vertex Shader, inputType=FunctionInput_StaticBool, sortPriority=9}
   ))
 ]}
