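# Prototype Landmass material function that previews landscape terrain and water simulation from height, normal and velocity render targets.
# Outputs base color, roughness, emissive, normals and height-based WPO, with water gated by MPC_Erosion's ShowWater.
# Water uses refracted UVs, exponential absorption and scattering, Fresnel sky reflection, flowmap foam and spiral-blurred normals.
# Emissive also overlays a brush-preview highlight, velocity debug display and red warning for out-of-range heights.
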MaterialFunction /Landmass/Landscape/OldPrototype_BP/Mats/MF/PreviewLSWaterSim
graph {nodes [
  n1 = MaterialFunctionCall() {materialFunction=&/PreviewLSCoords}
  n2 = VertexInterpolator(VS=@n1.Result)
  n3 = ScalarParameter() {parameterName=NormalBias}
  n4 = TextureSampleParameter2D(UVs=@n2.PS) {parameterName=NormalRT, samplerSource=SSM_Wrap_WorldGroupSettings, texture=&Black}
  n5 = ConstantBiasScale(None=AppendVector(A=@n4.RGB, B=@n4.RGB)) {bias=-0.5, scale=2}
  n6 = Add(A=AppendVector(A=Divide(A=@n3) {constB=1.618}, B=@n3), B=ComponentMask(None=@n5) {a=true, b=true})
  n7 = MaterialFunctionCall!(
   Refractive Index 1
   Refractive Index 2=ScalarParameter() {defaultValue=1.33, parameterName=n2}
   Surface Normal=DeriveNormalZ(InXY=@n6)
   Camera Vector=CameraVectorWS()
  ) {materialFunction=&/RefractionWorldVector}
  n8 = Multiply(A=Divide(A=ComponentMask(None=@n7.Refracted Vector) {g=true, r=true}, B=ComponentMask(None=@n7.Refracted Vector) {b=true})) {constB=-1}
  n9 = TextureSampleParameter2D(UVs=@n2.PS) {parameterName=HeightRT, samplerSource=SSM_Wrap_WorldGroupSettings, texture=&Black}
  n10 = Max(A=@n9.G, B=@n9.G)
  n11 = Divide(A=Multiply(A=@n10), B=CollectionParameter() {collection=/Landmass/Landscape/BlueprintBrushes/MPC/MPC_Erosion, parameterName=TotalSizeX})
  n12 = ScalarParameter() {defaultValue=1, parameterName=Depth}
  n13 = Add(A=@n2.PS, B=Multiply(A=@n8, B=Multiply(A=@n11, B=@n12)))
  n14 = TextureSampleParameter2D!(UVs=@n13, Level=Constant(), Apply View MipBias) {mipValueMode=TMVM_MipLevel, parameterName=NormalRT, samplerSource=SSM_Wrap_WorldGroupSettings, texture=&Black}
  ComponentMask(None=ConstantBiasScale(None=AppendVector(A=@n14.RGB, B=@n14.RGB)) {bias=-0.5, scale=2}) {g=true, r=true}
  n15 = Multiply(A=CollectionParameter() {collection=&MPC_Erosion, parameterName=ShowWater}, B=@n10)
  MaterialExpressionMaterialFunctionCall_5 = MaterialFunctionCall(Vector 2=Multiply(A=@n15, B=@n8)) {materialFunction=/Engine/Functions/Engine_MaterialFunctions02/Utility/VectorLength}
  MaterialExpressionDivide_2 = Divide(
   A=ScalarParameter() {defaultValue=8, parameterName=Density Scale}
   B=VectorParameter() {parameterName=Exct, defaultValue=<rgba 250 5000 227000 0>}.RGB
  )
  n16 = Custom!(t=@MaterialExpressionMaterialFunctionCall_5.V2 Length, c=@MaterialExpressionDivide_2) {
   code="return exp(-t * c);"
   description=Exponential Density
   inputs [
    {
     inputName=t
     input {expression=&self:MaterialExpressionMaterialFunctionCall_5, outputIndex=1}
    }
    {inputName=c, input {expression=&self:MaterialExpressionDivide_2}}
   ]
  }
  MaterialExpressionDivide_12 = Divide(A=@MaterialExpressionMaterialFunctionCall_5.V2 Length) {constB=100}
  MaterialExpressionScalarParameter_23 = ScalarParameter() {defaultValue=0.02, parameterName=Scatter Density}
  n17 = Custom(t=@MaterialExpressionDivide_12, c=@MaterialExpressionScalarParameter_23) {
   code="return exp(-t * c);"
   description=Exponential Density
   inputs [
    {inputName=t, input {expression=&self:MaterialExpressionDivide_12}}
    {inputName=c, input {expression=&self:MaterialExpressionScalarParameter_23}}
   ]
  }
  n18 = Multiply(A=@n16, B=@n17) {desc=* Scattering Opacity}
  n19 = DeriveNormalZ(InXY=@n6)
  n20 = Fresnel(ExponentIn=ScalarParameter() {defaultValue=3, parameterName=FresnelExp}, Normal=@n19)
  n21 = Saturate(None=Multiply(A=@n15, B=ScalarParameter() {defaultValue=0.02, parameterName=WaterRoughMult}))
  n22 = LinearInterpolate(B=OneMinus(None=@n20), Alpha=@n21) {constA=1}
  FunctionOutput(None=Multiply(A=FunctionInput() {bUsePreviewValueAsDefault=true, inputName=BaseColor, previewValue=<1 1 1 1>}, B=Multiply(A=@n18, B=@n22))) {outputName=BaseColor}
  FunctionOutput(None=LinearInterpolate(A=ScalarParameter() {defaultValue=1, parameterName=LandR}, B=ScalarParameter() {defaultValue=1, parameterName=WaterR}, Alpha=@n21)) {outputName=Roughness, sortPriority=1}
  n23 = TextureSampleParameter2D!(UVs=@n2.PS, Apply View MipBias) {parameterName=HeightRT, samplerSource=SSM_Wrap_WorldGroupSettings, texture=&Black}
  n24 = Max(A=@n23.G, B=@n23.G)
  n25 = Subtract(A=@n24, B=ScalarParameter() {defaultValue=50, parameterName=asdfasf})
  n26 = ScalarParameter() {defaultValue=1000, parameterName=fasdfasferwer}
  n27 = TextureSampleParameter2D(UVs=@n2.PS) {parameterName=VelocityRT, samplerSource=SSM_Wrap_WorldGroupSettings, texture=&Black}
  n28 = Divide(A=@n27.RGB, B=CollectionParameter() {collection=&MPC_Erosion, parameterName=GridSize})
  n29 = Multiply(A=@n28, B=ScalarParameter() {defaultValue=0.05, parameterName=Flowmap Amount})
  n30 = VectorParameter() {parameterName=asdfa, defaultValue=<rgba 1 -1 0 0>}
  n31 = ScalarParameter() {defaultValue=8000, parameterName=TT}
  n32 = Multiply(
   A=Saturate(None=Divide(A=@n25, B=@n26) {constB=50})
   B=MaterialFunctionCall!(
    Diffuse=TextureObject() {texture=/Landmass/PreviewContent/Water/Textures/T_Foam_DistanceField_Fade_02}
    Normal Map
    "Flow Vector Map (see tooltip)"=Multiply(A=@n29, B=@n30.RGB)
    UV texture 2 offset
    "Time (see tooltip)"
    UVs=MaterialFunctionCall!(Scale=@n31) {materialFunction=/Engine/Functions/Engine_MaterialFunctions01/Coordinates/WorldPosition-XY}.Result
   ) {materialFunction=/Engine/Functions/Engine_MaterialFunctions02/Texturing/FlowMaps_Simple}.Diffuse
  )
  n33 = ScalarParameter() {parameterName=Foam}
  n34 = CollectionParameter() {collection=&MPC_Erosion, parameterName=ShowWater}
  MaterialExpressionMultiply_22 = Multiply(
   A=MaterialFunctionCall(Vector 2=@n28) {materialFunction=&VectorLength}.V2 Length
   B=ScalarParameter() {defaultValue=0.5, parameterName=asdfadre}
  )
  n35 = Custom!(x=@MaterialExpressionMultiply_22) {
   code="return 1-exp(-x);"
   description=exp
   inputs [{inputName=x, input {expression=&self:MaterialExpressionMultiply_22}}]
  }
  n36 = Multiply(A=@n33, B=Multiply(A=@n33, B=Multiply(A=@n34, B=@n35)))
  n37 = TextureSampleParameter2D(UVs=@n2.PS) {parameterName=PreviewBrushHeightAlpha, samplerSource=SSM_Wrap_WorldGroupSettings, texture=&Black}
  n38 = Abs(None=Subtract(A=@n23.R, B=@n37.R))
  n39 = ScalarParameter() {defaultValue=2048, parameterName=swt}
  n40 = OneMinus(None=Saturate(None=Divide(A=@n38, B=@n39) {constB=5000}))
  n41 = Constant3Vector() {constant=<rgba 0 0.25 0 0>}
  n42 = ScalarParameter() {defaultValue=0.25, parameterName=Sel}
  n43 = Abs(None=Divide(
    A=Multiply(A=@n27.RGB, B=@n24)
    B=CollectionParameter() {collection=&MPC_Erosion, parameterName=TotalSizeX}
   ))
  n44 = CollectionParameter() {collection=&MPC_Erosion, parameterName=VelocityDisplay}
  n45 = Add(
   A=Multiply(A=Min(A=@n40, B=@n37.G), B=Multiply(A=@n41, B=@n42))
   B=Multiply(
    A=Multiply(A=@n43, B=@n44)
    B=CollectionParameter() {collection=&MPC_Erosion, parameterName=ShowWater}
   )
  )
  n46 = Multiply(
   A=@n20
   B=TextureSample(UVs=MaterialFunctionCall!(Normal=@n19, CameraVector) {materialFunction=/Engine/Functions/Engine_MaterialFunctions02/Utility/CustomReflectionVector}.Reflection Vector) {samplerType=SAMPLERTYPE_LinearColor, texture=/Landmass/PreviewContent/Sky/Tex/Sky54a2048}.RGB
  )
  n47 = Multiply(A=ScalarParameter() {defaultValue=0.5, parameterName=Cubemap Brightness}, B=AtmosphericLightColor())
  n48 = VectorParameter() {parameterName=Scattering Color, defaultValue=<rgba 0.1244 0.9701 1 0.1>}
  n49 = Multiply(A=OneMinus(None=@n17), B=Multiply(A=@n48.RGB, B=@n48.RGB))
  n50 = AtmosphericLightVector()
  n51 = ConstantBiasScale(None=DotProduct(A=@n7.Refracted Vector, B=Multiply(A=@n50) {constB=-1}))
  n52 = Constant3Vector() {constant=<rgba 1 0 0 0>}
  n53 = Subtract(A=@n9.R) {constB=32767}
  n54 = Subtract(A=Multiply(A=@n9.R) {constB=-1}) {constB=32767}
  n55 = Max(A=Saturate(None=Ceil(None=@n53)), B=Saturate(None=Ceil(None=@n54)))
  n56 = Add(A=Multiply(A=@n21, B=Multiply(A=@n46, B=@n47)), B=Add(A=Multiply(A=@n49, B=@n51), B=Multiply(A=@n52, B=@n55)))
  FunctionOutput(None=Add(A=Multiply(A=@n32, B=@n36), B=Add(A=@n45, B=@n56))) {outputName=Emissive, sortPriority=2}
  n57 = ComponentMask(None=@n5) {g=true, r=true}
  n58 = MaterialFunctionCall!(
   Vector to Transform=FunctionInput() {bUsePreviewValueAsDefault=true, inputName=Normal, sortPriority=1, previewValue=<0 0 1 1>}
   Z Vector=Multiply(A=DeriveNormalZ(InXY=@n57), B=Constant3Vector() {constant=<rgba -1 -1 1 0>})
   Center Location
  ) {materialFunction=/Engine/Functions/Engine_MaterialFunctions02/Math/TransformToZVector}
  n59 = ScalarParameter() {defaultValue=0.0003, parameterName=BlurDist}
  n60 = ScalarParameter() {defaultValue=3, parameterName=Steps}
  n61 = MaterialFunctionCall!(
   TextureObject=TextureObjectParameter() {parameterName=NormalRT, texture=&Black}
   UV=@n13
   Distance=Add(A=Multiply(A=@n59, B=@MaterialExpressionDivide_12), B=ScalarParameter() {defaultValue=0.0001, parameterName=mindist})
   Distance Steps=@n60
   Radial Steps=@n60
   Radial Offset
   Kernel Power
  ) {materialFunction=/Engine/Functions/Engine_MaterialFunctions01/Texturing/SpiralBlur-Texture}
  n62 = Add(
   A=ConstantBiasScale(None=ComponentMask(None=@n61.Result) {g=true, r=true}) {bias=-0.5, scale=2}
   B=AppendVector(A=Constant(), B=ScalarParameter() {defaultValue=0.01, parameterName=NormalBias_1})
  )
  FunctionOutput(None=LinearInterpolate(A=@n58.Result, B=DeriveNormalZ(InXY=@n62), Alpha=Saturate(None=@n15))) {outputName=Water and Terrain Normal, sortPriority=4}
  FunctionOutput(None=@n58.Result) {outputName=Terrain Normal, sortPriority=3}
  n63 = TextureSampleParameter2D(UVs=@n1.Result) {parameterName=HeightRT, samplerSource=SSM_Wrap_WorldGroupSettings, texture=&Black}
  n64 = Add(
   A=@n63.R
   B=Multiply(
    A=Max(A=@n63.G, B=@n63.G)
    B=CollectionParameter() {collection=&MPC_Erosion, parameterName=ShowWater}
   )
  )
  n65 = CollectionParameter() {collection=&MPC_Erosion, parameterName=ZScale}
  FunctionOutput(None=Multiply(A=Multiply(A=@n64, B=@n65), B=Constant3Vector() {constant=<rgba 0 0 1 0>})) {outputName=WPO Terrain + Water, sortPriority=5}
  n66 = CollectionParameter() {collection=&MPC_Erosion, parameterName=ZScale}
  FunctionOutput(None=Multiply(A=Multiply(A=@n63.R, B=@n66), B=Constant3Vector() {constant=<rgba 0 0 1 0>})) {outputName=WPO Terrain, sortPriority=6}
 ]}
