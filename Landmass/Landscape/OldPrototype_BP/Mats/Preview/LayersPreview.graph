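# Preview material from the old Landmass landscape prototype that shows layered Rock, Grass, Sand and Snow terrain using material attributes.
# Grass replaces rock on flatter ground through a parameterized threshold on the Z of the water-sim terrain normal.
# The Biomes texture then adds sand from red and snow from blue, sampled with preview landscape coordinates.
# Chained LayerHeightBlend functions combine the layers before PreviewLSAttributes processes the result.
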Material /Landmass/Landscape/OldPrototype_BP/Mats/Preview/LayersPreview
graph {
 properties {bUseMaterialAttributes=true, bUsedWithInstancedStaticMeshes=true, pixelDepthOffsetMode=PDOM_Legacy, usedShadingModels=Default Lit}
 nodes [
  n1 = MaterialFunctionCall!(BaseColor, Normal) {materialFunction=/Landmass/Landscape/OldPrototype_BP/Mats/MF/PreviewLSWaterSim}
  n2 = ScalarParameter() {defaultValue=0.2, parameterName=Grass Z Length}
  n3 = Subtract(
   A=ConstantBiasScale(None=ComponentMask(None=@n1.Terrain Normal) {b=true})
   B=Add(A=ScalarParameter() {defaultValue=0.875, parameterName=Grass Z Min}, B=Divide(A=@n2) {constB=-2})
  )
  n4 = MaterialFunctionCall!(
   Bottom=MaterialFunctionCall!(Material Attributes) {materialFunction=/Landmass/PreviewContent/MatLayers/Rock}.Material Attributes
   Top=MaterialFunctionCall!(Material Attributes) {materialFunction=/Landmass/PreviewContent/MatLayers/Grass}.Material Attributes
   Alpha=Saturate(None=Divide(A=@n3, B=@n2))
   Edge Tint Color
   Height Multiplier
   Contrast
   Blend Vertex Shader
  ) {materialFunction=/Landmass/Landscape/OldPrototype_BP/Mats/MF/LayerHeightBlend}
  n5 = MaterialFunctionCall!(Material Attributes) {materialFunction=/Landmass/PreviewContent/MatLayers/Sand}
  n6 = TextureSampleParameter2D(UVs=MaterialFunctionCall() {materialFunction=/Landmass/Landscape/OldPrototype_BP/Mats/MF/PreviewLSCoords}.Result) {parameterName=Biomes, texture=&Black}
  n7 = VectorParameter() {parameterName=SandEdgeTint, defaultValue=<rgba 0.5 0 0 0>}
  MaterialFunctionCall!(
   Attributes=MaterialFunctionCall(
    Bottom=MaterialFunctionCall(Bottom=@n4.Result, Top=@n5.Material Attributes, Alpha=@n6.R, Edge Tint Color=@n7.RGB) {materialFunction=&LayerHeightBlend}.Result
    Top=MaterialFunctionCall!(Material Attributes) {materialFunction=/Landmass/PreviewContent/MatLayers/Snow}.Material Attributes
    Alpha=@n6.B
   ) {materialFunction=&LayerHeightBlend}.Result
   Show Fluid Sim
  ) {materialFunction=/Landmass/Landscape/OldPrototype_BP/Mats/MF/PreviewLSAttributes}
  MaterialFunctionCall() {materialFunction=&PreviewLSAttributes}
 ]
}
