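# Landscape material in the Landmass old Blueprint prototype that previews Grass, Rock, Sand and Snow layers on terrain.
# Each layer is a material-attributes function, with Grass as the base.
# Chained LayerHeightBlend calls add Rock, then Sand, then Snow, each using its landscape layer sample as alpha.
# The Sand blend also exposes a SandEdgeTint color parameter for its edge.
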Material /Landmass/Landscape/OldPrototype_BP/Mats/Preview/LayersPreview_Landscape
graph {
 properties {
  bUseMaterialAttributes=true
  bUsedWithInstancedStaticMeshes=true
  bUsedWithStaticLighting=true
  pixelDepthOffsetMode=PDOM_Legacy
  usedShadingModels=Default Lit
 }
 nodes [
  n1 = MaterialFunctionCall!(Material Attributes) {materialFunction=/Landmass/PreviewContent/MatLayers/Grass}
  n2 = MaterialFunctionCall!(Material Attributes) {materialFunction=/Landmass/PreviewContent/MatLayers/Rock}
  n3 = LandscapeLayerSample() {parameterName=Rock}
  MaterialFunctionCall!(
   Bottom=MaterialFunctionCall(
    Bottom=MaterialFunctionCall(Bottom=@n1.Material Attributes, Top=@n2.Material Attributes, Alpha=@n3) {materialFunction=/Landmass/Landscape/OldPrototype_BP/Mats/MF/LayerHeightBlend}.Result
    Top=MaterialFunctionCall!(Material Attributes) {materialFunction=/Landmass/PreviewContent/MatLayers/Sand}.Material Attributes
    Alpha=LandscapeLayerSample() {parameterName=Sand}
    Edge Tint Color=VectorParameter() {parameterName=SandEdgeTint, defaultValue=<rgba 0.5 0 0 0>}.RGB
   ) {materialFunction=&LayerHeightBlend}.Result
   Top=MaterialFunctionCall!(Material Attributes) {materialFunction=/Landmass/PreviewContent/MatLayers/Snow}.Material Attributes
   Alpha=LandscapeLayerSample() {parameterName=Snow}
   Edge Tint Color
   Height Multiplier
   Contrast
   Blend Vertex Shader
  ) {materialFunction=&LayerHeightBlend}
 ]
}
