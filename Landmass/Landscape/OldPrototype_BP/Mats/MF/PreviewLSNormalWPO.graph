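# Material function for previewing Landmass landscape render targets, part of an old Blueprint prototype, outputting a terrain normal and a world position offset.
# UVs from PreviewLSCoords sample the HeightRT and NormalRT texture parameters, with a VertexInterpolator passing them to the normal sample.
# Normal RG is remapped to signed range and Z is derived.
# Height R is scaled by ZScale from MPC_Erosion and applied along Z as the offset.
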MaterialFunction /Landmass/Landscape/OldPrototype_BP/Mats/MF/PreviewLSNormalWPO
graph {nodes [
  n1 = MaterialFunctionCall() {materialFunction=&/PreviewLSCoords}
  n2 = TextureSampleParameter2D(UVs=VertexInterpolator(VS=@n1.Result).PS) {parameterName=NormalRT, samplerSource=SSM_Wrap_WorldGroupSettings, texture=&Black}
  n3 = ConstantBiasScale(None=AppendVector(A=@n2.RGB, B=@n2.RGB)) {bias=-0.5, scale=2}
  FunctionOutput(None=DeriveNormalZ(InXY=ComponentMask(None=@n3) {g=true, r=true})) {outputName=Terrain Normal, sortPriority=3}
  n4 = TextureSampleParameter2D(UVs=@n1.Result) {parameterName=HeightRT, samplerSource=SSM_Wrap_WorldGroupSettings, texture=&Black}
  n5 = CollectionParameter() {collection=/Landmass/Landscape/BlueprintBrushes/MPC/MPC_Erosion, parameterName=ZScale}
  FunctionOutput(None=Multiply(A=Multiply(A=@n4.R, B=@n5), B=Constant3Vector() {constant=<rgba 0 0 1 0>})) {outputName=WPO Terrain, sortPriority=6}
 ]}
