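# Lit material for the old Landmass landscape prototype that previews terrain on meshes, including instanced static meshes, using world-space normals.
# A Biomes texture, sampled at landscape preview coordinates, blends base, farm, desert, snow and mountain colors into a water-sim preview function.
# A separate function supplies preview normal and WPO logic.
# A debug branch draws a world-space component grid and numeric readout from MPC_Erosion, with grid lines toggled by ShowComponents.
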Material /Landmass/Landscape/OldPrototype_BP/Mats/Preview/MeshPreview
graph {
 properties {bTangentSpaceNormal=false, bUsedWithInstancedStaticMeshes=true, pixelDepthOffsetMode=PDOM_Legacy, usedShadingModels=Default Lit}
 nodes [
  n1 = ComponentMask(None=WorldPosition().XYZ) {g=true, r=true}
  n2 = CollectionParameter() {collection=/Landmass/Landscape/BlueprintBrushes/MPC/MPC_Erosion, parameterName=ComponentRes}
  n3 = CollectionParameter() {collection=&MPC_Erosion, parameterName=GridSize}
  n4 = Multiply(A=Subtract(A=@n2), B=@n3)
  n5 = MaterialFunctionCall!(
   A=Frac(None=Divide(A=@n1, B=@n4))
   B=Constant() {r=0.5}
   Bounds=ScalarParameter() {defaultValue=0.997, parameterName=B2}
   Edge Falloff=ScalarParameter() {parameterName=ef}
  ) {materialFunction=/Engine/Functions/Engine_MaterialFunctions01/Gradient/BoxMask-2D}
  Add(
   A=Multiply(
    A=OneMinus(None=@n5.Result)
    B=CollectionParameter() {collection=&MPC_Erosion, parameterName=ShowComponents}
   )
   B=MaterialFunctionCall!(Vector2=AppendVector(A=@n2, B=@n3), MaximumNumberOfDigits, UVs, DebugTextLocation RG_UpperRight BA_LowerLeft, Component Spacing) {materialFunction=/Engine/Functions/Engine_MaterialFunctions02/Utility/DebugFloat2Values}.ColorCodedOutput
  )
  MaterialFunctionCall() {materialFunction=/Landmass/Landscape/OldPrototype_BP/Mats/MF/PreviewLSNormalWPO}
  n6 = TextureSampleParameter2D(UVs=VertexInterpolator(VS=MaterialFunctionCall() {materialFunction=/Landmass/Landscape/OldPrototype_BP/Mats/MF/PreviewLSCoords}.Result).PS) {parameterName=Biomes, texture=&Black}
  n7 = LinearInterpolate(
   A=LinearInterpolate(
    A=ScalarParameter() {defaultValue=0.185, parameterName=Base}
    B=VectorParameter() {parameterName=Farm, defaultValue=<rgba 0.07 0.1719 0.0638 1>}.RGB
    Alpha=Saturate(None=@n6.G)
   )
   B=VectorParameter() {parameterName=DesertColor, defaultValue=<rgba 0.9688 0.4848 0.2069 1>}.RGB
   Alpha=Saturate(None=@n6.R)
  )
  n8 = VectorParameter() {parameterName=Snow, defaultValue=<rgba 0.8 0.9 0.9 1>}
  n9 = Saturate(None=@n6.B)
  n10 = Saturate(None=@n6.A)
  MaterialFunctionCall!(
   BaseColor=LinearInterpolate(
    A=LinearInterpolate(A=@n7, B=@n8.RGB, Alpha=@n9)
    B=VectorParameter() {parameterName=Mountains, defaultValue=<rgba 0.1 0.1 0.1 1>}.RGB
    Alpha=OneMinus(None=@n10)
   )
   Normal
  ) {materialFunction=/Landmass/Landscape/OldPrototype_BP/Mats/MF/PreviewLSWaterSim}
  ScalarParameter() {defaultValue=1, parameterName=Opacity}
  ScalarParameter() {defaultValue=0.25, parameterName=Spec}
 ]
}
