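# Preview sand material for mesh rendering in the old Landmass landscape Blueprint prototype.
# It uses Material Attributes, fed entirely by a single call to the Sand material layer function from Landmass preview content.
# It is Default Lit and enabled for instanced static meshes, so sand shading logic lives in the shared function rather than in this material.
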Material /Landmass/Landscape/OldPrototype_BP/Mats/Preview/SandMesh
graph {
 properties {bUseMaterialAttributes=true, bUsedWithInstancedStaticMeshes=true, pixelDepthOffsetMode=PDOM_Legacy, usedShadingModels=Default Lit}
 nodes [MaterialFunctionCall!(Material Attributes) {materialFunction=/Landmass/PreviewContent/MatLayers/Sand}]
}
